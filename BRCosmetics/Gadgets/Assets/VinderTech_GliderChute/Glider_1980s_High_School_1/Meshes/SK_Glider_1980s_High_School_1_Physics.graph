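# Physics asset for the 1980s High School glider skeletal mesh, setting up a jointed antenna for physical simulation.
# A capsule on the Body bone anchors a chain of four identical thin capsules on antenna bones 01 through 04.
# Each segment is constrained to the one before it with tight swing limits and very little twist, so the antenna can bend along its length while barely rotating about its axis.
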PhysicsAsset /BRCosmetics/Gadgets/Assets/VinderTech_GliderChute/Glider_1980s_High_School_1/Meshes/SK_Glider_1980s_High_School_1_Physics
bodies {
 Body [{shapeType=Capsule, length=10, radius=10}]
 antenna_01 [{shapeType=Capsule, length=19.5435, radius=0.9886, center=<-0.1148 0.2559 8.9869>, rotation=<rot 0 -90 0>}]
 antenna_02 [{shapeType=Capsule, length=19.5435, radius=0.9886, center=<-0.1148 0.2559 8.9869>, rotation=<rot 0 -90 0>}]
 antenna_03 [{shapeType=Capsule, length=19.5435, radius=0.9886, center=<-0.1148 0.2559 8.9869>, rotation=<rot 0 -90 0>}]
 antenna_04 [{shapeType=Capsule, length=19.5435, radius=0.9886, center=<-0.1148 0.2559 8.9869>, rotation=<rot 0 -90 0>}]
}
constraints [
 {bone1Name=antenna_02, bone2Name=antenna_01, swing1LimitDegrees=15, swing2LimitDegrees=15, twistLimitDegrees=2}
 {bone1Name=antenna_03, bone2Name=antenna_02, swing1LimitDegrees=15, swing2LimitDegrees=15, twistLimitDegrees=2}
 {bone1Name=antenna_04, bone2Name=antenna_03, swing1LimitDegrees=15, swing2LimitDegrees=15, twistLimitDegrees=2}
 {bone1Name=antenna_01, bone2Name=Body, swing1LimitDegrees=15, swing2LimitDegrees=15, twistLimitDegrees=2}
]
