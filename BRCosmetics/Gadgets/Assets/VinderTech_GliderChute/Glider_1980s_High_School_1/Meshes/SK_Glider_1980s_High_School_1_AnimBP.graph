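# Post-process style Anim Blueprint for the 1980s High School glider skeleton, adding physics-driven secondary motion to the incoming pose.
# The input pose passes through a RigidBody node using the glider's own physics asset, with overridden world gravity and amplified component linear velocity.
# The ParaGlide_Lean curve drives simulation alpha inverted, so physics influence fades as lean rises.
# The event graph holds no logic.
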AnimBlueprint /BRCosmetics/Gadgets/Assets/VinderTech_GliderChute/Glider_1980s_High_School_1/Meshes/SK_Glider_1980s_High_School_1_AnimBP
target_skeleton=&/SK_Glider_1980s_High_School_1_Skeleton
graphs [
 AnimGraph {nodes [
   AnimGraphNode_SubInput_0 = Input Pose()
   n1 = RigidBody!(ComponentPose=Local To Component(LocalPose=@AnimGraphNode_SubInput_0), AlphaCurveName:Name=ParaGlide_Lean, ExternalForce) -> (Pose) {
    alphaInputType=Curve
    bOverrideWorldGravity=true
    overridePhysicsAsset=&/SK_Glider_1980s_High_School_1_Physics
    alphaBoolBlend {blendOption=Linear}
    alphaScaleBiasClamp {bias=1, scale=-1}
    componentLinearVelScale=<2 2 2>
   }
   Output Pose(Result=Component To Local(ComponentPose=@n1)) {layerGroup=None}
  ]}
 EventGraph {nodes [
   TryGetPawnOwner()
   Event BlueprintUpdateAnimation()
  ]}
]
variables []
